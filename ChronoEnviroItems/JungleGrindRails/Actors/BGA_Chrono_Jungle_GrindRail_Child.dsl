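; Jungle vine variant of a grind rail that grants fall-damage immunity, picks meshes by segment length and sets vine materials.
; In EditMode or via SpawnVines, seeded random line traces from the spline find static, non-damageable or landscape anchors.
; Targets tagged as connect targets snap vines to random "Vine" sockets without connector pieces.
; Stored transforms become instanced vines and connectors, while a separate function scatters non-colliding moss on non-mobile platforms.
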(event EventBeginPlay
  (|CallParentFunction) ; This node is disabled and will not be called.\nDrag off pins to build functionality.
)

(event Collision|EventActorBeginOverlap
  (params (OtherActor :type Actor%20Object%20Reference))
  (|CallParentFunction OtherActor) ; This node is disabled and will not be called.\nDrag off pins to build functionality.
)

(event EventTick
  (params (DeltaSeconds :type Float%20%28single-precision%29)))

(event Custom|SpawnVines
  (Development|Editor|CreateCopyforUndoBuffer self)
  (CallFunction|SpawnAllSuportVines)
  (CallFunction|SpawnVineInstances))

(fn AddMossInstances
  (params (MossHISM :type Instanced%20Static%20Mesh%20Component%20Object%20Reference) (Scale :type Vector))
  (bind _returnvalue 5)
  (bind _returnvalue_2 512.0)
  (if (Utilities|IsValid MossHISM :!class /Script/Engine.KismetSystemLibrary)
    (Components|InstancedStaticMesh|ClearInstances MossHISM))
  (Collision|SetCollisionEnabled MossHISM "NoCollision")
  (LOD|SetMaxDrawDistance MossHISM 6e+03)
  (for _index (range 3 (+ (Math|Float|Truncate (* (/ (Spline|GetSplineLength (Variables|Default|GetSpline)) _returnvalue_2) _returnvalue)) 1))
    (Variables|Default|SetSpace (/ _returnvalue_2 _returnvalue))
    (bind _returnvalue_3 (CallFunction|CreateMossTransform _index Scale))
    (Components|InstancedStaticMesh|AddInstance MossHISM _returnvalue_3 false)))

(fn ConnectToSocket
  (params (Actor :type Actor%20Object%20Reference) (Stream :type Random%20Stream%20Structure%20%28by%20ref%29))
  (Utilities|FlowControl|Sequence
    (:then_0
      (for _array_element (Actor|GetComponentsByClass Actor "/Script/Engine.StaticMeshComponent")
        (for _array_element_2 (Transformation|GetAllSocketNames _array_element)
          (if (Utilities|String|Contains (Utilities|String|ToString(Name) _array_element_2) "Vine")
            (Utilities|Array|Add (LocalVariables|GetLocationList) (.location (Transformation|GetSocketTransform _array_element _array_element_2 "RTS_World")))))))
    (:then_1
      (return (Utilities|Array|Get(acopy) (LocalVariables|GetLocationList) (Math|Random|RandomIntegerinRangefromStream Stream 0 (Utilities|Array|LastIndex (LocalVariables|GetLocationList)))) (>= (Utilities|Array|LastIndex (LocalVariables|GetLocationList)) 0)))))

(fn CreateMossTransform
  (params (A :type Integer) (Scale :type Vector))
  (return (Math|Transform|MakeTransform (- (Spline|GetLocationatDistanceAlongSpline (Variables|Default|GetSpline) (+ (* A (Variables|Default|GetSpace)) (Math|Random|RandomFloatinRange -7.0 7.0)) "Local") (Math|Vector|MakeVector (Math|Random|RandomFloatinRange -7.0 7.0) (Math|Random|RandomFloatinRange -7.0 7.0) (- (* (Math|Float|Absolute(Float) (Math|Random|RandomFloatinRange -7.0 7.0)) 1.0) 15.0))) (Math|Rotator|CombineRotators (Spline|GetRotationatDistanceAlongSpline (Variables|Default|GetSpline) (+ (* A (Variables|Default|GetSpace)) (Math|Random|RandomFloatinRange -7.0 7.0)) "Local") (Math|Vector|RotatorfromAxisandAngle (Spline|GetTangentatDistanceAlongSpline (Variables|Default|GetSpline) (+ (* A (Variables|Default|GetSpace)) (Math|Random|RandomFloatinRange -7.0 7.0)) "Local") (* (Math|Random|RandomFloatinRange -7.0 7.0) 1e+01))) Scale)))

(fn GetMeshForSplineLength
  (params (Segment%20Length :type Float%20%28double-precision%29))
  (bind _array_element (Utilities|Array|ForEachLoopwithBreak (Variables|Default|GetLengthToMeshData))
    (:LoopBody
      (bind (_array_element_mesh_5_c75f5aec4289371b608ab796fe7dbab7 _array_element_maxlengthtousefor_2_1b54df844d592ffc6c68bd99036c99d7) (Utilities|Struct|BreakFortVineRailMeshData _array_element))
      (if (<= Segment%20Length _array_element_maxlengthtousefor_2_1b54df844d592ffc6c68bd99036c99d7)
        (LocalVariables|SetFoundMesh _array_element_mesh_5_c75f5aec4289371b608ab796fe7dbab7)))
    (:Completed
      _))
  (Utilities|IsValid (LocalVariables|GetFoundMesh)
    (:Is%20Valid
      _)
    (:Is%20Not%20Valid
      (bind _meshtouse (|CallParentFunction Segment%20Length))
      (bind _output_get (LocalVariables|SetFoundMesh _meshtouse))))
  (return _output_get))

(fn SetSplineMeshMaterials
  (params (SplineMeshComp :type Spline%20Mesh%20Component%20Object%20Reference))
  ;[1 1 1] Set material for the spline-thickend vines
  (Rendering|Material|SetMaterial SplineMeshComp 0 (Variables|Material|GetSplineVineMaterial))
  (Rendering|Material|SetMaterial SplineMeshComp 1 (Variables|Material|GetMaterialOfEmisssive)))

(fn ShouldApplyFallDamageImmunity
  (return true))

(fn SpawnAllSuportVines
  (Variables|Default|SetEditMode true)
  (Utilities|Array|Clear (Variables|Default|GetSupportVines_ConnectorsList))
  (Utilities|Array|Clear (Variables|Default|GetSupportVinesList))
  (Components|InstancedStaticMesh|ClearInstances (Variables|Default|GetConnectionsVines_HISM))
  (CallFunction|SpawnSupportVinesStartEnd false)
  (CallFunction|SpawnSupportVinesStartEnd true)
  (CallFunction|SpawnSupportVinesMiddle false))

(fn SpawnMoss
  (if (not (TouchPlatform|IsMobileGame))
    (Utilities|FlowControl|Sequence
      (:then_0
        (CallFunction|AddMossInstances (Variables|Default|GetMoss_HISM) (Variables|Default|GetMossScale))))))

(fn SpawnSupportVines_Middle
  (params (IsEnd :type Boolean))
  (bind _spline (Variables|Default|GetSpline))
  (bind _returnvalue (Spline|GetSplineLength _spline))
  (bind _stream (Variables|Default|GetStream))
  (bind _returnvalue_2 (Math|Random|RandomFloatinRangefromStream _stream 0.0 3e+02))
  (bind _step_size (LocalVariables|GetStep%20Size))
  (bind _returnvalue_3 (/ _returnvalue _step_size))
  (bind _returnvalue_4 (Math|Float|Floor _returnvalue_3))
  (bind _truestepsize (LocalVariables|GetTrueStepSize))
  (bind _overridehitlocation (LocalVariables|GetOverrideHitLocation))
  (bind _overridehit_location (LocalVariables|GetOverrideHit_Location))
  (Math|Random|SetRandomStreamSeed (Variables|Default|GetStream) (Variables|Default|GetSeed))
  (bind _index (Utilities|FlowControl|ForLoopwithBreak 1 _returnvalue_4)
    (:LoopBody
      (LocalVariables|SetTrueStepSize (/ _returnvalue _returnvalue_4))
      (LocalVariables|SetCount 0)
      (Utilities|FlowControl|ForLoopwithBreak 0 (LocalVariables|GetMaxTracesPerPoint)
        (:LoopBody
          (LocalVariables|SetSecondary false)
          (LocalVariables|SetisvalidHit false)
          (bind _output_get (LocalVariables|SetStartPoint (Spline|GetLocationatDistanceAlongSpline _spline (- (* _truestepsize _index) _returnvalue_2) "World")))
          (bind (_outhit _returnvalue_5) (Collision|LineTraceByChannel :Start _output_get :End (+ (Spline|GetLocationatDistanceAlongSpline _spline (- (* _truestepsize _index) _returnvalue_2) "World") (* (Math|Random|RandomUnitVectorinConeinDegreesfromStream (Variables|Default|GetStream) (Math|Vector|Lerp(Vector) (Math|Vector|SelectVector (Math|Vector|CrossProduct (Spline|FindDirectionClosesttoWorldLocation _spline (Spline|GetLocationatDistanceAlongSpline _spline (- (* _truestepsize _index) _returnvalue_2) "World") "World") (Spline|GetUpVectoratDistanceAlongSpline _spline (- (* _truestepsize _index) _returnvalue_2) "Local")) (Math|Vector|NegateVector (Math|Vector|CrossProduct (Spline|FindDirectionClosesttoWorldLocation _spline (Spline|GetLocationatDistanceAlongSpline _spline (- (* _truestepsize _index) _returnvalue_2) "World") "World") (Spline|GetUpVectoratDistanceAlongSpline _spline (- (* _truestepsize _index) _returnvalue_2) "Local"))) (>= (Math|Random|RandomFloatfromStream (Variables|Default|GetStream)) 0.5)) (Math|Vector|SelectVector (Math|Vector|NegateVector (Spline|FindDirectionClosesttoWorldLocation _spline (Spline|GetLocationatDistanceAlongSpline _spline (- (* _truestepsize _index) _returnvalue_2) "World") "World")) (Spline|FindDirectionClosesttoWorldLocation _spline (Spline|GetLocationatDistanceAlongSpline _spline (- (* _truestepsize _index) _returnvalue_2) "World") "World") (>= (Math|Random|RandomFloatfromStream (Variables|Default|GetStream)) 0.5)) (Math|Random|RandomFloatinRangefromStream (Variables|Default|GetStream) 0.2 0.75)) (Math|Float|SelectFloat 5.0 45.0 (LocalVariables|GetSecondary))) (LocalVariables|GetMaxVineLength))) :TraceChannel "TraceTypeQuery1" :bTraceComplex true :DrawDebugType "ForDuration" :bIgnoreSelf true :TraceColor "(R=1.000000,G=0.000000,B=0.000000,A=1.000000)" :TraceHitColor "(R=0.000000,G=1.000000,B=0.000000,A=1.000000)" :DrawTime 1.0))
          (LocalVariables|SetOverrideHitLocation false)
          (LocalVariables|SetSpawnConnector true)
          (Utilities|FlowControl|Sequence
            (:then_0
              _)
            (:then_1
              ;[1 1 1] Check for valid actors that are not destructible
              (bind _asbuilding_actor (Utilities|Casting|CastToBuildingActor (.HitActor (Collision|BreakHitResult _outhit))))
              (LocalVariables|SetisvalidHit (not (Variables|/Script/Engine.Actor|GetbCanBeDamaged _asbuilding_actor))))
            (:then_2
              (Utilities|Casting|CastToLandscapeProxy (.HitActor (Collision|BreakHitResult _outhit)))
              (if (not (> (Math|Vector|DotProduct (.Normal (Collision|BreakHitResult _outhit)) "0.000000,0.000000,1.000000") (Math|Random|RandomFloatinRangefromStream (Variables|Default|GetStream) 0.5 1.0)))
                (LocalVariables|SetisvalidHit true)))
            (:then_3
              (if (Actor|ActorHasTag (.HitActor (Collision|BreakHitResult _outhit)) "Jungle.GrindRail.ConnectTarget")
                (bind (_output _valid) (CallFunction|ConnecttoSocket (.HitActor (Collision|BreakHitResult _outhit)) (Variables|Default|GetStream)))
                (if _valid
                  (LocalVariables|SetOverrideHit_Location _output)
                  (LocalVariables|SetOverrideHitLocation true)
                  (else
                    (LocalVariables|SetisvalidHit false)))
                (LocalVariables|SetSpawnConnector false)))
            (:then_5
              (if (and _returnvalue_5 (LocalVariables|GetisvalidHit))
                (Math|Integer|IncrementInt (LocalVariables|GetCount)
                  (:%20%20
                    (bind _returnvalue_6 (Utilities|Array|Add (Variables|Default|GetSupportVinesList) (Math|Transform|MakeTransform (LocalVariables|GetStartPoint) (Math|Rotator|CombineRotators (Math|Rotator|FindLookatRotation (LocalVariables|GetStartPoint) (select _overridehitlocation _overridehit_location (.Location (Collision|BreakHitResult _outhit)))) (Math|Vector|RotatorfromAxisandAngle (Math|Vector|GetForwardVector (Math|Rotator|FindLookatRotation (LocalVariables|GetStartPoint) (select _overridehitlocation _overridehit_location (.Location (Collision|BreakHitResult _outhit))))) (Math|Random|RandomFloatinRangefromStream (Variables|Default|GetStream) 0.0 45.0))) (Math|Vector|MakeVector (/ (Math|Vector|Distance(Vector) (LocalVariables|GetStartPoint) (select _overridehitlocation _overridehit_location (.Location (Collision|BreakHitResult _outhit)))) 5e+02) (/ (Math|Vector|Distance(Vector) (LocalVariables|GetStartPoint) (select _overridehitlocation _overridehit_location (.Location (Collision|BreakHitResult _outhit)))) 5e+02) (Math|Float|Lerp (Math|Random|RandomFloatinRangefromStream (Variables|Default|GetStream) 1e+01 3e+01) (Math|Random|RandomFloatinRangefromStream (Variables|Default|GetStream) 5e+01 1e+02) (Math|Float|Clamp(Float) (/ (Math|Vector|Distance(Vector) (LocalVariables|GetStartPoint) (select _overridehitlocation _overridehit_location (.Location (Collision|BreakHitResult _outhit)))) 1e+03)))))))
                    (if (LocalVariables|GetSpawnConnector)
                      (Utilities|Array|Add (Variables|Default|GetSupportVines_ConnectorsList) (Math|Transform|MakeTransform (+ (select _overridehitlocation _overridehit_location (.Location (Collision|BreakHitResult _outhit))) (+ (* (Math|Vector|GetUnitDirection(Vector) (select _overridehitlocation _overridehit_location (.Location (Collision|BreakHitResult _outhit))) (LocalVariables|GetStartPoint)) (* 65.0 (.x (.scale (Utilities|Array|Get(acopy) (Variables|Default|GetSupportVinesList) _returnvalue_6))))) (Utilities|Operators|Multiply (Math|Vector|NegateVector (Math|Vector|GetUpVector (.rotation (Utilities|Array|Get(acopy) (Variables|Default|GetSupportVinesList) _returnvalue_6)))) (.z (.scale (Utilities|Array|Get(acopy) (Variables|Default|GetSupportVinesList) _returnvalue_6))) 0.41))) (Math|Rotator|CombineRotators (Math|Rotator|MakeRotfromX (Math|Vector|GetUnitDirection(Vector) (+ (select _overridehitlocation _overridehit_location (.Location (Collision|BreakHitResult _outhit))) (+ (* (Math|Vector|GetUnitDirection(Vector) (select _overridehitlocation _overridehit_location (.Location (Collision|BreakHitResult _outhit))) (LocalVariables|GetStartPoint)) (* 65.0 (.x (.scale (Utilities|Array|Get(acopy) (Variables|Default|GetSupportVinesList) _returnvalue_6))))) (Utilities|Operators|Multiply (Math|Vector|NegateVector (Math|Vector|GetUpVector (.rotation (Utilities|Array|Get(acopy) (Variables|Default|GetSupportVinesList) _returnvalue_6)))) (.z (.scale (Utilities|Array|Get(acopy) (Variables|Default|GetSupportVinesList) _returnvalue_6))) 0.41))) (select _overridehitlocation _overridehit_location (.Location (Collision|BreakHitResult _outhit))))) (Math|Vector|RotatorfromAxisandAngle (Math|Vector|GetForwardVector (Math|Rotator|MakeRotfromX (Math|Vector|GetUnitDirection(Vector) (+ (select _overridehitlocation _overridehit_location (.Location (Collision|BreakHitResult _outhit))) (+ (* (Math|Vector|GetUnitDirection(Vector) (select _overridehitlocation _overridehit_location (.Location (Collision|BreakHitResult _outhit))) (LocalVariables|GetStartPoint)) (* 65.0 (.x (.scale (Utilities|Array|Get(acopy) (Variables|Default|GetSupportVinesList) _returnvalue_6))))) (Utilities|Operators|Multiply (Math|Vector|NegateVector (Math|Vector|GetUpVector (.rotation (Utilities|Array|Get(acopy) (Variables|Default|GetSupportVinesList) _returnvalue_6)))) (.z (.scale (Utilities|Array|Get(acopy) (Variables|Default|GetSupportVinesList) _returnvalue_6))) 0.41))) (select _overridehitlocation _overridehit_location (.Location (Collision|BreakHitResult _outhit)))))) (Math|Random|RandomFloatinRangefromStream (Variables|Default|GetStream) 0.0 3.6e+02))) (Math|Vector|MakeVector (.x (.scale (Utilities|Array|Get(acopy) (Variables|Default|GetSupportVinesList) _returnvalue_6))) 1.0 1.0))))
                    (Utilities|FlowControl|Sequence
                      (:then_0
                        (if (>= (LocalVariables|GetCount) (LocalVariables|GetMaxVinesPerPoint))
                          _))))))))
          (Utilities|Casting|CastToStaticMeshActor (.HitActor (Collision|BreakHitResult _outhit)))
          (LocalVariables|SetisvalidHit true))))))

(fn SpawnSupportVines_StartEnd
  (params (IsEnd :type Boolean))
  (bind _spline (Variables|Default|GetSpline))
  (bind _returnvalue (Spline|GetSplineLength _spline))
  (bind _isend (Variables|GetIsEnd))
  (bind _stream (Variables|Default|GetStream))
  (bind _returnvalue_2 (Math|Random|RandomFloatinRangefromStream _stream 2e+03 1e+03))
  (bind _returnvalue_3 (- _returnvalue _returnvalue_2))
  (bind _returnvalue_4 (Math|Float|SelectFloat _returnvalue_3 _returnvalue_2 _isend))
  (bind _returnvalue_5 (Spline|GetLocationatDistanceAlongSpline _spline _returnvalue_4 "World"))
  (bind _supportvineslist (Variables|Default|GetSupportVinesList))
  (bind _overridehitlocation (LocalVariables|GetOverrideHitLocation))
  (bind _overridehit_location (LocalVariables|GetOverrideHit_Location))
  (Math|Random|SetRandomStreamSeed (Variables|Default|GetStream))
  (LocalVariables|SetCount 0)
  (Utilities|FlowControl|ForLoopwithBreak 0 (LocalVariables|GetMaxTracesPerPoint)
    (:LoopBody
      (LocalVariables|SetisvalidHit false)
      (bind _output_get (LocalVariables|SetStartPoint _returnvalue_5))
      (bind (_outhit _returnvalue_6) (Collision|LineTraceByChannel :Start _output_get :End (+ _returnvalue_5 (* (Math|Random|RandomUnitVectorinConeinDegreesfromStream (Variables|Default|GetStream) (Math|Vector|SelectVector (Spline|FindDirectionClosesttoWorldLocation _spline _returnvalue_5 "World") (Math|Vector|NegateVector (Spline|FindDirectionClosesttoWorldLocation _spline _returnvalue_5 "World")) (Variables|GetIsEnd)) 45.0) (+ _returnvalue_2 5e+02))) :TraceChannel "TraceTypeQuery10" :bTraceComplex true :DrawDebugType "ForDuration" :bIgnoreSelf true :TraceColor "(R=1.000000,G=0.000000,B=0.000000,A=1.000000)" :TraceHitColor "(R=0.000000,G=1.000000,B=0.000000,A=1.000000)" :DrawTime 5.0))
      (LocalVariables|SetOverrideHitLocation false)
      (LocalVariables|SetSpawnConnector true)
      (Utilities|FlowControl|Sequence
        (:then_0
          _)
        (:then_1
          ;[1 1 1] Check for valid actors that are not destructible
          (bind _asbuilding_actor (Utilities|Casting|CastToBuildingActor (.HitActor (Collision|BreakHitResult _outhit))))
          (LocalVariables|SetisvalidHit (not (Variables|/Script/Engine.Actor|GetbCanBeDamaged _asbuilding_actor))))
        (:then_2
          (Utilities|Casting|CastToLandscapeProxy (.HitActor (Collision|BreakHitResult _outhit)))
          (if (not (> (Math|Vector|DotProduct "0, 0, 0" "0.000000,0.000000,1.000000") (Math|Random|RandomFloatinRangefromStream (Variables|Default|GetStream) 0.5 1.0)))
            (LocalVariables|SetisvalidHit true)))
        (:then_3
          (if (Actor|ActorHasTag (.HitActor (Collision|BreakHitResult _outhit)) "Jungle.GrindRail.ConnectTarget")
            (bind (_output _valid) (CallFunction|ConnecttoSocket (.HitActor (Collision|BreakHitResult _outhit)) (Variables|Default|GetStream)))
            (if _valid
              (LocalVariables|SetOverrideHit_Location _output)
              (LocalVariables|SetOverrideHitLocation true)
              (else
                (LocalVariables|SetisvalidHit false)))
            (LocalVariables|SetSpawnConnector false)))
        (:then_5
          (if (and _returnvalue_6 (LocalVariables|GetisvalidHit))
            (Math|Integer|IncrementInt (LocalVariables|GetCount)
              (:%20%20
                (bind _returnvalue_7 (Utilities|Array|Add _supportvineslist (Math|Transform|MakeTransform (LocalVariables|GetStartPoint) (Math|Rotator|FindLookatRotation (LocalVariables|GetStartPoint) (select _overridehitlocation _overridehit_location (.Location (Collision|BreakHitResult _outhit)))) (Math|Vector|MakeVector (/ (Math|Vector|Distance(Vector) (LocalVariables|GetStartPoint) (select _overridehitlocation _overridehit_location (.Location (Collision|BreakHitResult _outhit)))) 5e+02) (/ (Math|Vector|Distance(Vector) (LocalVariables|GetStartPoint) (select _overridehitlocation _overridehit_location (.Location (Collision|BreakHitResult _outhit)))) 5e+02) (Math|Random|RandomFloatinRangefromStream (Variables|Default|GetStream) 5e+01 1e+02)))))
                (if (LocalVariables|GetSpawnConnector)
                  (Utilities|Array|Add (Variables|Default|GetSupportVines_ConnectorsList) (Math|Transform|MakeTransform (+ (select _overridehitlocation _overridehit_location (.Location (Collision|BreakHitResult _outhit))) (+ (* (Math|Vector|GetUnitDirection(Vector) (select _overridehitlocation _overridehit_location (.Location (Collision|BreakHitResult _outhit))) (LocalVariables|GetStartPoint)) (* 65.0 (.x (.scale (Utilities|Array|Get(acopy) _supportvineslist _returnvalue_7))))) (Utilities|Operators|Multiply (Math|Vector|NegateVector (Math|Vector|GetUpVector (.rotation (Utilities|Array|Get(acopy) _supportvineslist _returnvalue_7)))) (.z (.scale (Utilities|Array|Get(acopy) _supportvineslist _returnvalue_7))) 0.41))) (Math|Rotator|CombineRotators (Math|Rotator|MakeRotfromX (Math|Vector|GetUnitDirection(Vector) (+ (select _overridehitlocation _overridehit_location (.Location (Collision|BreakHitResult _outhit))) (+ (* (Math|Vector|GetUnitDirection(Vector) (select _overridehitlocation _overridehit_location (.Location (Collision|BreakHitResult _outhit))) (LocalVariables|GetStartPoint)) (* 65.0 (.x (.scale (Utilities|Array|Get(acopy) _supportvineslist _returnvalue_7))))) (Utilities|Operators|Multiply (Math|Vector|NegateVector (Math|Vector|GetUpVector (.rotation (Utilities|Array|Get(acopy) _supportvineslist _returnvalue_7)))) (.z (.scale (Utilities|Array|Get(acopy) _supportvineslist _returnvalue_7))) 0.41))) (select _overridehitlocation _overridehit_location (.Location (Collision|BreakHitResult _outhit))))) (Math|Vector|RotatorfromAxisandAngle (Math|Vector|GetForwardVector (Math|Rotator|MakeRotfromX (Math|Vector|GetUnitDirection(Vector) (+ (select _overridehitlocation _overridehit_location (.Location (Collision|BreakHitResult _outhit))) (+ (* (Math|Vector|GetUnitDirection(Vector) (select _overridehitlocation _overridehit_location (.Location (Collision|BreakHitResult _outhit))) (LocalVariables|GetStartPoint)) (* 65.0 (.x (.scale (Utilities|Array|Get(acopy) _supportvineslist _returnvalue_7))))) (Utilities|Operators|Multiply (Math|Vector|NegateVector (Math|Vector|GetUpVector (.rotation (Utilities|Array|Get(acopy) _supportvineslist _returnvalue_7)))) (.z (.scale (Utilities|Array|Get(acopy) _supportvineslist _returnvalue_7))) 0.41))) (select _overridehitlocation _overridehit_location (.Location (Collision|BreakHitResult _outhit)))))) (Math|Random|RandomFloatinRangefromStream (Variables|Default|GetStream) 0.0 3.6e+02))) (Math|Vector|MakeVector (.x (.scale (Utilities|Array|Get(acopy) _supportvineslist _returnvalue_7))) 1.0 1.0))))
                (Utilities|FlowControl|Sequence
                  (:then_0
                    (if (>= (LocalVariables|GetCount) (LocalVariables|GetMaxVinesPerPoint))
                      _))))))))
      (Utilities|Casting|CastToStaticMeshActor (.HitActor (Collision|BreakHitResult _outhit)))
      (LocalVariables|SetisvalidHit true))))

(fn SpawnVineInstances
  (Utilities|FlowControl|Sequence
    (:then_0
      (for _array_element (Variables|Default|GetSupportVinesList)
        (Components|InstancedStaticMesh|AddInstance (Variables|Default|GetConnectionsVines_HISM) _array_element true)))
    (:then_1
      (for _array_element_2 (Variables|Default|GetSupportVines_ConnectorsList)
        (Components|InstancedStaticMesh|AddInstance (Variables|Default|GetConnectionsVines_Connector_HISM) _array_element_2 true)))))

(fn ConstructionScript
  (|CallParentFunction)
  (Utilities|FlowControl|Sequence
    (:then_0
      (if (Variables|Default|GetEditMode)
        (CallFunction|SpawnAllSuportVines)))
    (:then_1
      (CallFunction|SpawnVineInstances))))
