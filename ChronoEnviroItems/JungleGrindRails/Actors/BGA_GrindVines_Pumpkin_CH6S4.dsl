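; Pumpkin variant of a grind vine actor for jungle grind rails, inheriting its behavior from a parent class.
; Its construction script runs the parent's, sets the default scene root to Static mobility, then attaches the spline component to that root with relative transforms kept.
; BeginPlay and ActorBeginOverlap hold only disabled parent-call stubs, so this child adds no runtime logic.
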(event EventBeginPlay
  (|CallParentFunction) ; This node is disabled and will not be called.\nDrag off pins to build functionality.
)

(event Collision|EventActorBeginOverlap
  (params (OtherActor :type Actor%20Object%20Reference))
  (|CallParentFunction OtherActor) ; This node is disabled and will not be called.\nDrag off pins to build functionality.
)

(fn ConstructionScript
  (bind _defaultsceneroot (Variables|Default|GetDefaultSceneRoot))
  (|CallParentFunction)
  (Transformation|SetMobility _defaultsceneroot "Static")
  (Transformation|AttachComponentToComponent (Variables|Default|GetSpline) _defaultsceneroot "None" "KeepRelative" "KeepRelative" "KeepRelative" true))
