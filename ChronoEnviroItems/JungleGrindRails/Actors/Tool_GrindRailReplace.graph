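# Editor utility widget that transfers spline shapes from grind rails to pumpkin grind vine actors through Copy, Clean and Paste buttons.
# Copy caches the selected BGA_GrindRail's spline points in local space, including tangents, rotation, scale and type.
# Paste replaces the selected BGA_GrindVines_Pumpkin_CH6S4's spline with the cached points and calls GenerateRailComponents.
# Clean clears that vine's spline and repeatedly removes mesh components until none remain.
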EditorUtilityWidgetBlueprint /ChronoEnviroItems/JungleGrindRails/Actors/Tool_GrindRailReplace.Tool_GrindrailReplace
graphs [EventGraph {nodes [
   "On Clicked (EditorUtilityButton_Copy)"() {is=title K2Node_ComponentBoundEvent}
   > Clear(TargetArray:Array of Spline Point Structures=Get Spline Points() -> (Spline Points:Array of Spline Point Structures))
   > n1 = GetSelectedActors()
   > n2 = Cast To BGA_GrindRail!(execute, Object="Get (a copy)"(Array:Array of Actor Object References=@n1) -> (Output:Actor Object Reference)) -> (then, CastFailed, AsBGA Grind Rail:BGA Grind Rail Object Reference) {is=title K2Node_DynamicCast}
   n3 = Get Spline!(self:BGA Grind Rail Object Reference=@n2) -> (Spline:Spline Component Object Reference)
   > n4 = For Loop(LastIndex=int - int(A=GetNumberOfSplinePoints!(self:Spline Component Object Reference=@n3) -> (ReturnValue:int=0)))
   n5 = GetTransformAtSplinePoint!(self:Spline Component Object Reference=@n3, PointIndex:int=@n4, CoordinateSpace:ESplineCoordinateSpace Enum=Local, bUseScale:bool=true) -> (ReturnValue_Location:Vector, ReturnValue_Rotation:Rotator, ReturnValue_Scale:Vector)
   n6 = Get Spline Points()
   > Add!(
    execute
    TargetArray:Array of Spline Point Structures=@n6
    NewItem_InputKey:float="To Float (Integer)"(InInt=@n4)
    NewItem_Position:Vector=@n5
    NewItem_ArriveTangent:Vector=GetTangentAtSplinePoint!(self:Spline Component Object Reference=@n3, PointIndex:int=@n4, CoordinateSpace:ESplineCoordinateSpace Enum=Local) -> (ReturnValue:Vector)
    NewItem_LeaveTangent:Vector=GetLeaveTangentAtSplinePoint!(self:Spline Component Object Reference=@n3, PointIndex:int=@n4, CoordinateSpace:ESplineCoordinateSpace Enum=Local) -> (ReturnValue:Vector)
    NewItem_Rotation:Rotator=@n5.ReturnValue_Rotation
    NewItem_Scale:Vector=@n5.ReturnValue_Scale
    NewItem_Type:ESplinePointType Enum=GetSplinePointType!(self:Spline Component Object Reference=@n3, PointIndex:int=@n4) -> (ReturnValue:ESplinePointType Enum=Linear)
   ) -> (then, ReturnValue) {is=title K2Node_CallArrayFunction}
   > PrintText(InText="To Text (Integer)"(Value=Length(TargetArray:Array of Spline Point Structures=@n6)))
   PrintString(execute=[@n2.CastFailed, @n11.CastFailed, @n8.CastFailed], InString=Incorrect Spline)
   "On Clicked (EditorUtilityButton_Paste)"() {is=title K2Node_ComponentBoundEvent}
   > n7 = GetSelectedActors()
   > n8 = Cast To BGA_GrindVines_Pumpkin_CH6S4!(execute, Object="Get (a copy)"(Array=@n7)) -> (then, CastFailed, AsBGA Grind Vines Pumpkin CH6S4:BGA Grind Vines Pumpkin CH6S4 Object Reference) {is=title K2Node_DynamicCast}
   n9 = Get Spline(self=@n8)
   > ClearSplinePoints!(execute, self:Spline Component Object Reference=@n9, bUpdateSpline:bool=true) -> (then)
   > AddPoints!(execute, self:Spline Component Object Reference=@n9, Points:Array of Spline Point Structures=Get Spline Points(), bUpdateSpline:bool=false) -> (then)
   > UpdateSpline(self:Spline Component Object Reference=@n9)
   > GenerateRailComponents(self:BGA Grind Rail Object Reference=@n8)
   "On Clicked (EditorUtilityButton_Clean)"() {is=title K2Node_ComponentBoundEvent}
   > n10 = GetSelectedActors()
   > n11 = Cast To BGA_GrindVines_Pumpkin_CH6S4(Object="Get (a copy)"(Array=@n10))
   > ClearSplinePoints(self=Get Spline(self=@n11), bUpdateSpline=true)
   n12 = Get Spline(self=@n11)
   > ClearSplinePoints(self=@n12, bUpdateSpline=true)
   > UpdateSpline(self=@n12)
   > n13 = While Loop!(execute:Exec, Condition:bool="integer > integer"(A=GetMeshCount!(self:BGA Grind Rail Object Reference=@n11) -> (MeshCount:int=0))) -> (LoopBody:Exec, Completed:Exec) {is=title K2Node_MacroInstance}
   RemoveAllMeshComponents(execute=[@n13.Completed, @n13.LoopBody], self:BGA Grind Rail Object Reference=@n11)
   RedoSplineMesh(self:BGA Grind Rail Object Reference=@n8)
   Generate Meshes Along Spline(self:BGA Grind Rail Object Reference=@n8)
   Event PreConstruct()
   Event Construct()
   Event Tick()
  ]}]
variables [
 BPTransform {type=Vector}
 Spline Points {type="array<SplinePoint>", default []}
]
widgets {
 info {rootWidgetClass=&StackBox, widgetCount=7}
 widgets [
  StackBox_0 {widgetClassPath=&StackBox, widget=StackBox_0}
  EditorUtilityButton_Copy {
   widgetClassPath=&EditorUtilityButton
   parent=StackBox_0
   widget=EditorUtilityButton_Copy
   slot {class=StackBoxSlot, properties {size {sizeRule=Fill, value=1}}}
  }
  TextBlock_36 {widgetClassPath=&TextBlock, parent=EditorUtilityButton_Copy, widget=TextBlock_36, properties {text=Copy}, slot {class=ButtonSlot}}
  EditorUtilityButton_Clean {
   widgetClassPath=&EditorUtilityButton
   parent=StackBox_0
   widget=EditorUtilityButton_Clean
   slot {class=StackBoxSlot, properties {size {sizeRule=Fill, value=1}}}
  }
  Clean {widgetClassPath=&TextBlock, parent=EditorUtilityButton_Clean, widget=Clean, properties {text=Clean}, slot {class=ButtonSlot}}
  EditorUtilityButton_Paste {
   widgetClassPath=&EditorUtilityButton
   parent=StackBox_0
   widget=EditorUtilityButton_Paste
   slot {class=StackBoxSlot, properties {size {sizeRule=Fill, value=1}}}
  }
  TextBlock_76 {
   widgetClassPath=&TextBlock
   parent=EditorUtilityButton_Paste
   widget=TextBlock_76
   properties {text=Paste}
   slot {class=ButtonSlot}
  }
 ]
}
